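# Material function for glTF import that evaluates base color, opacity and opacity mask from a texture and factor inputs.
# It samples the texture, white by default, using caller-supplied UVs and explicit DDX/DDY derivatives with a shared clamp sampler.
# BaseColor multiplies texture RGB by the RGB factor and vertex color.
# Opacity is the saturated product of texture alpha and alpha factor, and OpacityMask subtracts AlphaCutoff from it.
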MaterialFunction /InterchangeAssets/gltf/MaterialFunctions/MF_BaseColor
graph {nodes [
  n1 = TextureSample!(
   UVs=FunctionInput() {bUsePreviewValueAsDefault=true, inputName=BaseColorTexture_UVs, inputType=FunctionInput_Vector2, sortPriority=3}
   Tex=FunctionInput(Preview=TextureObject() {texture=/InterchangeAssets/gltf/Textures/T_White_srgb}) {bUsePreviewValueAsDefault=true, inputName=BaseColorTexture, inputType=FunctionInput_Texture2D}
   "DDX(UVs)"=FunctionInput() {bUsePreviewValueAsDefault=true, inputName=BaseColorTexture_DDX, inputType=FunctionInput_Vector2, sortPriority=4}
   "DDY(UVs)"=FunctionInput() {bUsePreviewValueAsDefault=true, inputName=BaseColorTexture_DDY, inputType=FunctionInput_Vector2, sortPriority=5}
   Apply View MipBias
  ) {mipValueMode=TMVM_Derivative, samplerSource=SSM_Clamp_WorldGroupSettings}
  n2 = FunctionInput() {bUsePreviewValueAsDefault=true, inputName=BaseColorFactor_RGB, sortPriority=1, previewValue=<1 1 1 1>}
  FunctionOutput(None=Multiply(A=Multiply(A=@n1.RGB, B=@n2), B=VertexColor())) {outputName=BaseColor}
  n3 = FunctionInput() {bUsePreviewValueAsDefault=true, inputName=BaseColorFactor_A, inputType=FunctionInput_Scalar, sortPriority=2, previewValue=<1 1 1 1>}
  n4 = Saturate(None=Multiply(A=@n3, B=@n1.A))
  FunctionOutput(None=@n4) {outputName=Opacity, sortPriority=1}
  FunctionOutput(None=Subtract(A=@n4, B=FunctionInput() {bUsePreviewValueAsDefault=true, inputName=AlphaCutoff, inputType=FunctionInput_Scalar, sortPriority=9})) {outputName=OpacityMask, sortPriority=2}
 ]}
