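# MF_IOR is a glTF material function in InterchangeAssets that converts an index of refraction into a Specular value.
# It squares the ratio of one minus IOR to IOR plus one to get F0 reflectance.
# It divides F0 by 0.08 and saturates the result for the Specular output.
# A second output passes the IOR through unchanged, and the input defaults to 1.5.
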MaterialFunction /InterchangeAssets/gltf/MaterialFunctions/MF_IOR
graph {nodes [
  n1 = FunctionInput() {bUsePreviewValueAsDefault=true, inputName=IOR, inputType=FunctionInput_Scalar, previewValue=<1.5 0 0 0>}
  n2 = Divide(A=OneMinus(None=@n1), B=Add(A=@n1))
  n3 = Multiply(A=@n2, B=@n2) {desc=F0 Reflectance}
  FunctionOutput(None=Saturate(None=Divide(A=@n3) {constB=0.08})) {outputName=Specular}
  FunctionOutput(None=@n1) {outputName=ior, sortPriority=1}
 ]}
