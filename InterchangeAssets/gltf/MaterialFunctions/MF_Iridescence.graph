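# Material function for glTF materials imported through Interchange that applies thin-film iridescence to an input base color.
# The thickness texture's green channel blends between minimum and maximum thickness, is scaled by a pi-based factor, and feeds MF_IridescenceNoLUT with IridescenceIOR and a constant base F0.
# Its Fresnel result is mixed with base color, weighted by factor times the iridescence texture's red channel.
# The base IOR input is unconnected.
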MaterialFunction /InterchangeAssets/gltf/MaterialFunctions/MF_Iridescence
graph {nodes [
  FunctionInput() {
   bUsePreviewValueAsDefault=true
   description=BaseMaterial's IOR
   inputName=IOR
   inputType=FunctionInput_Scalar
   sortPriority=1
   previewValue=<1.5 0 0 1>
  }
  n1 = FunctionInput() {inputName=BaseColor}
  n2 = FunctionInput() {
   bUsePreviewValueAsDefault=true
   inputName=IridescenceThicknessMinimum
   inputType=FunctionInput_Scalar
   sortPriority=6
   previewValue=<100 0 0 1>
  }
  n3 = FunctionInput() {
   bUsePreviewValueAsDefault=true
   inputName=IridescenceThicknessMaximum
   inputType=FunctionInput_Scalar
   sortPriority=7
   previewValue=<400 0 0 1>
  }
  n4 = FunctionInput() {bUsePreviewValueAsDefault=true, inputName=IridescenceThicknessTexture_UVs, inputType=FunctionInput_Vector2, sortPriority=11}
  n5 = FunctionInput() {bUsePreviewValueAsDefault=true, inputName=IridescenceThicknessTexture_DDX, inputType=FunctionInput_Vector2, sortPriority=12}
  n6 = FunctionInput() {bUsePreviewValueAsDefault=true, inputName=IridescenceThicknessTexture_DDY, inputType=FunctionInput_Vector2, sortPriority=13}
  n7 = TextureSample!(
   UVs=@n4
   Tex=FunctionInput(Preview=TextureObject() {texture=/InterchangeAssets/gltf/Textures/T_White_srgb}) {bUsePreviewValueAsDefault=true, inputName=IridescenceThicknessTexture, inputType=FunctionInput_Texture2D, sortPriority=4}
   "DDX(UVs)"=@n5
   "DDY(UVs)"=@n6
   Apply View MipBias
  ) {mipValueMode=TMVM_Derivative, samplerSource=SSM_Clamp_WorldGroupSettings}
  n8 = FunctionInput() {bUsePreviewValueAsDefault=true, inputName=IridescenceIOR, inputType=FunctionInput_Scalar, sortPriority=2, previewValue=<1.3 0 0 1>}
  n9 = MaterialFunctionCall!(
   NormalWS
   BaseF0=Constant() {r=0.85}
   BaseF90
   Thickness=Multiply(
    A=LinearInterpolate(A=@n2, B=@n3, Alpha=@n7.G)
    B=Custom() {code="return PI * 0.0001\r\n;", description="Pi Factor ", outputType=CMOT_Float1, inputs []}
   ) {constB=0}
   n2=@n8
   UseComplexIOR
   n3
   k
  ) {materialFunction=/InterchangeAssets/Functions/MF_IridescenceNoLUT}
  n10 = FunctionInput() {bUsePreviewValueAsDefault=true, inputName=IridescenceFactor, inputType=FunctionInput_Scalar, sortPriority=5}
  n11 = TextureSample(
   UVs=FunctionInput() {bUsePreviewValueAsDefault=true, inputName=IridescenceTexture_UVs, inputType=FunctionInput_Vector2, sortPriority=8}
   Tex=FunctionInput(Preview=TextureObject() {texture=&T_White_srgb}) {bUsePreviewValueAsDefault=true, inputName=IridescenceTexture, inputType=FunctionInput_Texture2D, sortPriority=3}
   "DDX(UVs)"=FunctionInput() {bUsePreviewValueAsDefault=true, inputName=IridescenceTexture_DDX, inputType=FunctionInput_Vector2, sortPriority=9}
   "DDY(UVs)"=FunctionInput() {bUsePreviewValueAsDefault=true, inputName=IridescenceTexture_DDY, inputType=FunctionInput_Vector2, sortPriority=10}
  ) {mipValueMode=TMVM_Derivative, samplerSource=SSM_Clamp_WorldGroupSettings}
  FunctionOutput(None=LinearInterpolate(A=@n1, B=LinearInterpolate(A=@n1, B=@n9.Fresnel), Alpha=Multiply(A=@n10, B=@n11.R))) {description=BaseColor with Iridescence applied., outputName=BaseColor}
 ]}
