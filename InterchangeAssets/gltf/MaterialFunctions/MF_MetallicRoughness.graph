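# Material function for glTF import materials that turns a metallic-roughness texture and two scalar factors into Metallic and Roughness outputs.
# The texture's blue channel is multiplied by MetallicFactor and its green channel by RoughnessFactor.
# Sampling uses caller-supplied UVs with explicit DDX/DDY derivatives and the shared clamp sampler.
# A linear white texture and unit factors serve as input defaults.
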MaterialFunction /InterchangeAssets/gltf/MaterialFunctions/MF_MetallicRoughness
graph {nodes [
  n1 = TextureSample!(
   UVs=FunctionInput() {bUsePreviewValueAsDefault=true, inputName=MetallicRoughnessTexture_UVs, inputType=FunctionInput_Vector2, sortPriority=3}
   Tex=FunctionInput(Preview=TextureObject() {samplerType=SAMPLERTYPE_LinearColor, texture=/InterchangeAssets/gltf/Textures/T_White_Linear}) {bUsePreviewValueAsDefault=true, inputName=MetallicRoughnessTexture, inputType=FunctionInput_Texture2D}
   "DDX(UVs)"=FunctionInput() {bUsePreviewValueAsDefault=true, inputName=MetallicRoughnessTexture_DDX, inputType=FunctionInput_Vector2, sortPriority=4}
   "DDY(UVs)"=FunctionInput() {bUsePreviewValueAsDefault=true, inputName=MetallicRoughnessTexture_DDY, inputType=FunctionInput_Vector2, sortPriority=5}
   Apply View MipBias
  ) {mipValueMode=TMVM_Derivative, samplerSource=SSM_Clamp_WorldGroupSettings, samplerType=SAMPLERTYPE_LinearColor}
  n2 = FunctionInput() {bUsePreviewValueAsDefault=true, inputName=MetallicFactor, inputType=FunctionInput_Scalar, sortPriority=1, previewValue=<1 0 0 1>}
  FunctionOutput(None=Multiply(A=@n1.B, B=@n2)) {outputName=Metallic}
  n3 = FunctionInput() {bUsePreviewValueAsDefault=true, inputName=RoughnessFactor, inputType=FunctionInput_Scalar, sortPriority=2, previewValue=<1 0 0 1>}
  FunctionOutput(None=Multiply(A=@n3, B=@n1.G)) {outputName=Roughness, sortPriority=1}
 ]}
