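# Material function from the glTF Interchange assets that produces the normal map output for glTF materials.
# It samples the normal texture with supplied UVs and explicit DDX/DDY derivatives, using a shared clamp sampler and a generic normal map as the preview default.
# MF_RotateNormals_TS rotates the tangent-space normal by the SinCos input, then MF_ScaleNormal applies NormalScale to give the Normals output.
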MaterialFunction /InterchangeAssets/gltf/MaterialFunctions/MF_Normals
graph {nodes [
  n1 = TextureSample!(
   UVs=FunctionInput() {bUsePreviewValueAsDefault=true, inputName=NormalTexture_UVs, inputType=FunctionInput_Vector2, sortPriority=2}
   Tex=FunctionInput(Preview=TextureObject() {samplerType=SAMPLERTYPE_Normal, texture=/InterchangeAssets/gltf/Textures/T_Generic_N}) {bUsePreviewValueAsDefault=true, inputName=NormalTexture, inputType=FunctionInput_Texture2D}
   "DDX(UVs)"=FunctionInput() {bUsePreviewValueAsDefault=true, inputName=NormalTexture_DDX, inputType=FunctionInput_Vector2, sortPriority=3}
   "DDY(UVs)"=FunctionInput() {bUsePreviewValueAsDefault=true, inputName=NormalTexture_DDY, inputType=FunctionInput_Vector2, sortPriority=4}
   Apply View MipBias
  ) {mipValueMode=TMVM_Derivative, samplerSource=SSM_Clamp_WorldGroupSettings, samplerType=SAMPLERTYPE_Normal}
  n2 = FunctionInput() {bUsePreviewValueAsDefault=true, inputName=NormalTexture_SinCos, inputType=FunctionInput_Vector2, sortPriority=5, previewValue=<0 1 0 1>}
  n3 = FunctionInput() {bUsePreviewValueAsDefault=true, inputName=NormalScale, inputType=FunctionInput_Scalar, sortPriority=1, previewValue=<1 0 0 1>}
  FunctionOutput(None=MaterialFunctionCall!(
    Normal=MaterialFunctionCall!(NormalsTS=@n1.RGB, SinCos=@n2) {materialFunction=&/MF_RotateNormals_TS}.Result
    Scale=@n3
   ) {materialFunction=&/MF_ScaleNormal}.Result) {outputName=Normals}
 ]}
