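# MF_Occlusion is a glTF Interchange material function that computes an Occlusion value from a texture and a strength.
# It samples the OcclusionTexture input (linear color, defaulting to white) at the supplied UVs using explicit DDX/DDY derivative inputs and the shared clamp sampler.
# The red channel is blended with 1 by the saturated OcclusionStrength, so zero strength yields no occlusion and full strength yields the raw texture value.
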MaterialFunction /InterchangeAssets/gltf/MaterialFunctions/MF_Occlusion
graph {nodes [
  n1 = FunctionInput() {bUsePreviewValueAsDefault=true, inputName=OcclusionTexture_UVs, inputType=FunctionInput_Vector2, sortPriority=2}
  n2 = FunctionInput(Preview=TextureObject() {samplerType=SAMPLERTYPE_LinearColor, texture=/InterchangeAssets/gltf/Textures/T_White_Linear}) {bUsePreviewValueAsDefault=true, inputName=OcclusionTexture, inputType=FunctionInput_Texture2D}
  n3 = FunctionInput() {bUsePreviewValueAsDefault=true, inputName=OcclusionTexture_DDX, inputType=FunctionInput_Vector2, sortPriority=3}
  n4 = FunctionInput() {bUsePreviewValueAsDefault=true, inputName=OcclusionTexture_DDY, inputType=FunctionInput_Vector2, sortPriority=4}
  n5 = FunctionInput() {bUsePreviewValueAsDefault=true, inputName=OcclusionStrength, inputType=FunctionInput_Scalar, sortPriority=1, previewValue=<1 0 0 1>}
  FunctionOutput(None=LinearInterpolate(
    A=Constant() {r=1}
    B=TextureSample!(UVs=@n1, Tex=@n2, "DDX(UVs)"=@n3, "DDY(UVs)"=@n4, Apply View MipBias) {mipValueMode=TMVM_Derivative, samplerSource=SSM_Clamp_WorldGroupSettings, samplerType=SAMPLERTYPE_LinearColor}.R
    Alpha=Saturate(None=@n5)
   )) {outputName=Occlusion}
 ]}
