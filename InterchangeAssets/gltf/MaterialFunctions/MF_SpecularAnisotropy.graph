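# Material function for glTF imports that converts an anisotropy texture, strength and rotation into Anisotropy and Tangent outputs.
# The texture's blue channel times strength gives Anisotropy, while red and green are remapped into a direction vector.
# Custom HLSL rotates that direction and transforms it by a tangent, bitangent and normal basis.
# Tangent and default normal use fixed axes for tangent-space-normal materials, otherwise normalized world-space vertex vectors.
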MaterialFunction /InterchangeAssets/gltf/MaterialFunctions/MF_SpecularAnisotropy
graph {nodes [
  n1 = TextureSample!(
   UVs=FunctionInput() {bUsePreviewValueAsDefault=true, inputName=AnisotropyTexture_UVs, inputType=FunctionInput_Vector2, sortPriority=4}
   Tex=FunctionInput(Preview=TextureObject() {samplerType=SAMPLERTYPE_LinearColor, texture=/InterchangeAssets/gltf/Textures/T_Anisotropy_Direction_Linear}) {bUsePreviewValueAsDefault=true, inputName=AnisotropyTexture, inputType=FunctionInput_Texture2D}
   "DDX(UVs)"=FunctionInput() {bUsePreviewValueAsDefault=true, inputName=AnisotropyTexture_DDX, inputType=FunctionInput_Vector2, sortPriority=5}
   "DDY(UVs)"=FunctionInput() {bUsePreviewValueAsDefault=true, inputName=AnisotropyTexture_DDY, inputType=FunctionInput_Vector2, sortPriority=6}
   Apply View MipBias
  ) {mipValueMode=TMVM_Derivative, samplerSource=SSM_Clamp_WorldGroupSettings}
  FunctionOutput(None=Multiply(A=@n1.B, B=FunctionInput() {bUsePreviewValueAsDefault=true, inputName=AnisotropyStrength, inputType=FunctionInput_Scalar, sortPriority=1})) {outputName=Anisotropy}
  MaterialExpressionFunctionInput_0 = FunctionInput() {bUsePreviewValueAsDefault=true, inputName=AnisotropyRotation, inputType=FunctionInput_Scalar, sortPriority=2}
  MaterialExpressionVertexTangentWS_1 = VertexTangentWS()
  MaterialExpressionCustom_1 = Custom!(SurfaceTangent=@MaterialExpressionVertexTangentWS_1) {
   code="{\r\n##if MATERIAL_TANGENTSPACENORMAL\r\n  const float3 T = float3(1,0,0);\r\n##else\r\n  const float3 T = normalize(SurfaceTangent);\r\n##endif\r\n  return T;\r\n}"
   description=GetTangent
   showCode=true
   inputs [{
     inputName=SurfaceTangent
     input {expression=&self:MaterialExpressionVertexTangentWS_1}
    }]
  }
  MaterialExpressionVertexNormalWS_0 = VertexNormalWS()
  n2 = Custom(SurfaceTangent=@MaterialExpressionVertexNormalWS_0) {
   code="{\r\n##if MATERIAL_TANGENTSPACENORMAL\r\n  const float3 T = float3(0,0,1);\r\n##else\r\n  const float3 T = normalize(SurfaceTangent);\r\n##endif\r\n  return T;\r\n}"
   description=GetNormal
   showCode=true
   inputs [{
     inputName=SurfaceTangent
     input {expression=&self:MaterialExpressionVertexNormalWS_0}
    }]
  }
  MaterialExpressionFunctionInput_2 = FunctionInput(Preview=@n2) {bUsePreviewValueAsDefault=true, inputName=Normal, sortPriority=3, previewValue=<0 0 1 1>}
  n3 = ComponentMask(None=@n1.RGB) {g=true, r=true}
  MaterialExpressionAppendVector_0 = AppendVector(A=Subtract(A=Multiply(A=@n3) {constB=2}), B=Constant())
  n4 = Custom!(
   Rotation=@MaterialExpressionFunctionInput_0
   SurfaceTangent=@MaterialExpressionCustom_1
   SurfaceNormal=@MaterialExpressionFunctionInput_2
   Direction=@MaterialExpressionAppendVector_0
  ) {
   code="const float3 N = SurfaceNormal;\r\nconst float3 T = SurfaceTangent;\r\nconst float3 B = normalize(cross(SurfaceTangent, SurfaceNormal));\r\n\r\nconst float cosAngle = cos(Rotation);\r\nconst float sinAngle = sin(Rotation);\r\n\r\nDirection.xy = mul(float2x2(cosAngle, sinAngle, -sinAngle, cosAngle), normalize(Direction.xy));\r\n\r\nconst float3 Out = normalize(mul(float3x3(T, B, N), Direction));\r\n\r\nreturn Out;"
   description=GetAnisotropyTangent
   showCode=true
   inputs [
    {inputName=Rotation, input {expression=&self:MaterialExpressionFunctionInput_0}}
    {inputName=SurfaceTangent, input {expression=&self:MaterialExpressionCustom_1}}
    {
     inputName=SurfaceNormal
     input {expression=&self:MaterialExpressionFunctionInput_2}
    }
    {inputName=Direction, input {expression=&self:MaterialExpressionAppendVector_0}}
   ]
  }
  FunctionOutput(None=@n4) {outputName=Tangent}
 ]}
