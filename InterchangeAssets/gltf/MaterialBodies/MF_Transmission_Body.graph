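# Material function body for glTF transmissive materials, outputting surface attributes plus TransmissionColor for translucent rendering.
# Switchable texture inputs with UV transforms feed glTF helpers for base color, metallic-roughness, normals, occlusion, specular/IOR, anisotropy, iridescence and transmission.
# Transmission is faked by tinting temporally blurred scene color into Emissive, blacked out on ES feature levels, while base color darkens and metallic fades.
# A path-tracing switch instead zeroes emissive and overrides roughness and opacity.
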MaterialFunction /InterchangeAssets/gltf/MaterialBodies/MF_Transmission_Body
graph {
 properties {previewBlendMode=BLEND_Translucent}
 nodes [
  FunctionOutput(None=LinearInterpolate(B=NamedRerouteUsage(), Alpha=NamedRerouteUsage()) {constA=1}) {outputName=TransmissionColor}
  n1 = NamedRerouteUsage()
  n2 = Constant() {desc=10% srgb, r=0.006}
  n3 = NamedRerouteUsage()
  n4 = NamedRerouteUsage()
  n5 = NamedRerouteUsage()
  n6 = MaterialFunctionCall!(ior=NamedRerouteUsage(), BlurIntensity=NamedRerouteUsage(), Normals=NamedRerouteUsage(), DisableRefraction=StaticBool() {value=true}) {materialFunction=/InterchangeAssets/gltf/MaterialFunctions/MF_Temporal_Blur}
  n7 = NamedRerouteUsage()
  n8 = NamedRerouteUsage()
  n9 = Multiply(
   A=Multiply(A=Multiply(A=@n6.Blured Scene, B=@n7) {desc=BluredScene}, B=Power(Base=@n8) {constExponent=2.2, desc=gamma correct}) {constA=1, desc=Tinted blur}
   B=NamedRerouteUsage()
  )
  n10 = Constant3Vector()
  n11 = NamedRerouteUsage()
  n12 = Max(A=OneMinus(None=NamedRerouteUsage()), B=Power(Base=NamedRerouteUsage()) {constExponent=0.1})
  n13 = GetMaterialAttributes(None=MakeMaterialAttributes!(
    BaseColor=LinearInterpolate(A=@n1, B=@n2, Alpha=@n3) {constB=0, desc=Black when we fully transmit}
    Metallic=LinearInterpolate(A=@n4, Alpha=@n5) {constB=0, desc=No metallic when we fully transmit}
    Specular=NamedRerouteUsage()
    Roughness=NamedRerouteUsage()
    Anisotropy=NamedRerouteUsage()
    EmissiveColor=FeatureLevelSwitch(Default=@n9, ES2=@n10, ES3_1=@n10, SM5=@n9, SM6=@n9)
    Opacity=Min(A=@n11, B=@n12) {desc=alphaAsCoverage}
    OpacityMask
    Normal=NamedRerouteUsage()
    Tangent=NamedRerouteUsage()
    WorldPositionOffset
    SubsurfaceColor
    ClearCoat
    ClearCoatRoughness
    AmbientOcclusion=NamedRerouteUsage()
    Refraction=Constant() {r=1}
    CustomizedUVs_0
    CustomizedUVs_1
    CustomizedUVs_2
    CustomizedUVs_3
    CustomizedUVs_4
    CustomizedUVs_5
    CustomizedUVs_6
    CustomizedUVs_7
    PixelDepthOffset
    ShadingModel
    Displacement
   )) {attributeGetTypes [#1]}
  n14 = Constant3Vector()
  n15 = Multiply(A=@n13.Roughness, B=NamedRerouteUsage())
  n16 = Min(A=OneMinus(None=NamedRerouteUsage()), B=NamedRerouteUsage())
  n17 = BreakMaterialAttributes!(Attr=PathTracingQualitySwitch!(
    Normal=@n13.MaterialAttributes
    PathTraced=SetMaterialAttributes!(MaterialAttributes=@n13.MaterialAttributes, EmissiveColor=@n14, Roughness=@n15, Opacity=@n16) {attributeSetTypes [#2, #1, #3]}
   ))
  FunctionOutput(None=@n17.Anisotropy) {outputName=Anisotropy, sortPriority=5}
  FunctionOutput(None=@n17.Tangent) {outputName=Tangent, sortPriority=9}
  FunctionOutput(None=@n17.BaseColor) {outputName=BaseColor, sortPriority=1}
  FunctionOutput(None=@n17.Metallic) {outputName=Metallic, sortPriority=2}
  FunctionOutput(None=@n17.Specular) {outputName=Specular, sortPriority=3}
  FunctionOutput(None=@n17.Roughness) {outputName=Roughness, sortPriority=4}
  FunctionOutput(None=@n17.EmissiveColor) {outputName=EmissiveColor, sortPriority=6}
  FunctionOutput(None=@n17.Opacity) {outputName=Opacity, sortPriority=7}
  FunctionOutput(None=@n17.Normal) {outputName=Normal, sortPriority=8}
  FunctionOutput(None=@n17.AmbientOcclusion) {outputName=Occlusion, sortPriority=10}
  n18 = NamedRerouteUsage()
  n19 = NamedRerouteUsage()
  n20 = FunctionInput() {bUsePreviewValueAsDefault=true, inputName=IridescenceIOR, inputType=FunctionInput_Scalar, sortPriority=400, previewValue=<1.3 0 0 1>}
  n21 = FunctionInput(Preview=TextureObject() {texture=/InterchangeAssets/gltf/Textures/T_White_srgb}) {bUsePreviewValueAsDefault=true, inputName=IridescenceTexture, inputType=FunctionInput_Texture2D, sortPriority=401}
  n22 = FunctionInput(Preview=TextureObject() {texture=&T_White_srgb}) {bUsePreviewValueAsDefault=true, inputName=IridescenceThicknessTexture, inputType=FunctionInput_Texture2D, sortPriority=402}
  n23 = FunctionInput() {bUsePreviewValueAsDefault=true, inputName=IridescenceFactor, inputType=FunctionInput_Scalar, sortPriority=403}
  n24 = FunctionInput() {
   bUsePreviewValueAsDefault=true
   inputName=IridescenceThicknessMinimum
   inputType=FunctionInput_Scalar
   sortPriority=404
   previewValue=<100 0 0 1>
  }
  n25 = FunctionInput() {
   bUsePreviewValueAsDefault=true
   inputName=IridescenceThicknessMaximum
   inputType=FunctionInput_Scalar
   sortPriority=405
   previewValue=<400 0 0 1>
  }
  n26 = FunctionInput() {
   bUsePreviewValueAsDefault=true
   inputName=IridescenceTexture_Offset_X
   inputType=FunctionInput_Scalar
   sortPriority=406
   previewValue=<0 0 0 0>
  }
  n27 = FunctionInput() {
   bUsePreviewValueAsDefault=true
   inputName=IridescenceTexture_Offset_Y
   inputType=FunctionInput_Scalar
   sortPriority=407
   previewValue=<0 0 0 0>
  }
  n28 = FunctionInput() {
   bUsePreviewValueAsDefault=true
   inputName=IridescenceTexture_Scale_X
   inputType=FunctionInput_Scalar
   sortPriority=408
   previewValue=<1 0 0 0>
  }
  n29 = FunctionInput() {
   bUsePreviewValueAsDefault=true
   inputName=IridescenceTexture_Scale_Y
   inputType=FunctionInput_Scalar
   sortPriority=409
   previewValue=<1 0 0 0>
  }
  n30 = FunctionInput() {bUsePreviewValueAsDefault=true, description=Rad, inputName=IridescenceTexture_Rotation, inputType=FunctionInput_Scalar, sortPriority=410}
  n31 = FunctionInput() {bUsePreviewValueAsDefault=true, inputName=IridescenceTexture_TilingMethod, inputType=FunctionInput_Vector2, sortPriority=412}
  n32 = MaterialFunctionCall!(
   Offset_X=@n26
   Offset_Y=@n27
   Scale_X=@n28
   Scale_Y=@n29
   Rotation=@n30
   TexCoord=FunctionInput() {bUsePreviewValueAsDefault=true, inputName=IridescenceTexture_TexCoord, inputType=FunctionInput_Scalar, sortPriority=411}
   TilingMethod=@n31
  ) {materialFunction=/InterchangeAssets/gltf/MaterialFunctions/MF_TransformUVs}
  n33 = FunctionInput() {
   bUsePreviewValueAsDefault=true
   inputName=IridescenceThicknessTexture_Offset_X
   inputType=FunctionInput_Scalar
   sortPriority=413
   previewValue=<0 0 0 0>
  }
  n34 = FunctionInput() {
   bUsePreviewValueAsDefault=true
   inputName=IridescenceThicknessTexture_Offset_Y
   inputType=FunctionInput_Scalar
   sortPriority=414
   previewValue=<0 0 0 0>
  }
  n35 = FunctionInput() {
   bUsePreviewValueAsDefault=true
   inputName=IridescenceThicknessTexture_Scale_X
   inputType=FunctionInput_Scalar
   sortPriority=415
   previewValue=<1 0 0 0>
  }
  n36 = FunctionInput() {
   bUsePreviewValueAsDefault=true
   inputName=IridescenceThicknessTexture_Scale_Y
   inputType=FunctionInput_Scalar
   sortPriority=416
   previewValue=<1 0 0 0>
  }
  n37 = FunctionInput() {
   bUsePreviewValueAsDefault=true
   description=Rad
   inputName=IridescenceThicknessTexture_Rotation
   inputType=FunctionInput_Scalar
   sortPriority=417
  }
  n38 = FunctionInput() {bUsePreviewValueAsDefault=true, inputName=IridescenceThicknessTexture_TexCoord, inputType=FunctionInput_Scalar, sortPriority=418}
  n39 = FunctionInput() {bUsePreviewValueAsDefault=true, inputName=IridescenceThicknessTexture_TilingMethod, inputType=FunctionInput_Vector2, sortPriority=419}
  n40 = MaterialFunctionCall(Offset_X=@n33, Offset_Y=@n34, Scale_X=@n35, Scale_Y=@n36, Rotation=@n37, TexCoord=@n38, TilingMethod=@n39) {materialFunction=&MF_TransformUVs}
  NamedRerouteDeclaration(None=StaticSwitchParameter(
    True=MaterialFunctionCall!(
     BaseColor=@n18
     IOR=@n19
     IridescenceIOR=@n20
     IridescenceTexture=@n21
     IridescenceThicknessTexture=@n22
     IridescenceFactor=@n23
     IridescenceThicknessMinimum=@n24
     IridescenceThicknessMaximum=@n25
     IridescenceTexture_UVs=@n32.UVs
     IridescenceTexture_DDX=@n32.UVs
     IridescenceTexture_DDY=@n32.UVs
     IridescenceThicknessTexture_UVs=@n40.UVs
     IridescenceThicknessTexture_DDX=@n40.UVs
     IridescenceThicknessTexture_DDY=@n40.UVs
    ) {materialFunction=/InterchangeAssets/gltf/MaterialFunctions/MF_Iridescence}.BaseColor
    False=NamedRerouteUsage()
   ) {defaultValue=true, group=Configuration, parameterName=bHasIridescence, sortPriority=9}) {name=BC, nodeColor=<rgba 0.8516 1 0 1>}
  n41 = FunctionInput() {bUsePreviewValueAsDefault=true, inputName=MetallicFactor, inputType=FunctionInput_Scalar, sortPriority=26, previewValue=<1 0 0 1>}
  n42 = FunctionInput() {bUsePreviewValueAsDefault=true, inputName=RoughnessFactor, inputType=FunctionInput_Scalar, sortPriority=27, previewValue=<1 0 0 1>}
  n43 = FunctionInput() {
   bUsePreviewValueAsDefault=true
   inputName=MetallicRoughnessTexture_Offset_X
   inputType=FunctionInput_Scalar
   sortPriority=28
   previewValue=<0 0 0 0>
  }
  n44 = FunctionInput() {
   bUsePreviewValueAsDefault=true
   inputName=MetallicRoughnessTexture_Offset_Y
   inputType=FunctionInput_Scalar
   sortPriority=29
   previewValue=<0 0 0 0>
  }
  n45 = FunctionInput() {
   bUsePreviewValueAsDefault=true
   inputName=MetallicRoughnessTexture_Scale_X
   inputType=FunctionInput_Scalar
   sortPriority=30
   previewValue=<1 0 0 0>
  }
  n46 = FunctionInput() {
   bUsePreviewValueAsDefault=true
   inputName=MetallicRoughnessTexture_Scale_Y
   inputType=FunctionInput_Scalar
   sortPriority=31
   previewValue=<1 0 0 0>
  }
  n47 = FunctionInput() {
   bUsePreviewValueAsDefault=true
   description=Rad
   inputName=MetallicRoughnessTexture_Rotation
   inputType=FunctionInput_Scalar
   sortPriority=32
  }
  n48 = FunctionInput() {bUsePreviewValueAsDefault=true, inputName=MetallicRoughnessTexture_TexCoord, inputType=FunctionInput_Scalar, sortPriority=33}
  n49 = FunctionInput() {bUsePreviewValueAsDefault=true, inputName=MetallicRoughnessTexture_TilingMethod, inputType=FunctionInput_Vector2, sortPriority=34}
  n50 = MaterialFunctionCall(Offset_X=@n43, Offset_Y=@n44, Scale_X=@n45, Scale_Y=@n46, Rotation=@n47, TexCoord=@n48, TilingMethod=@n49) {materialFunction=&MF_TransformUVs}
  n51 = MaterialFunctionCall!(
   MetallicRoughnessTexture=FunctionInput(Preview=TextureObject() {samplerType=SAMPLERTYPE_LinearColor, texture=/InterchangeAssets/gltf/Textures/T_White_Linear}) {bUsePreviewValueAsDefault=true, inputName=MetallicRoughnessTexture, inputType=FunctionInput_Texture2D, sortPriority=25}
   MetallicFactor=NamedRerouteDeclaration(None=@n41) {name=Input_MetallicFactor, nodeColor=<rgba 1 0 0.0469 1>}
   RoughnessFactor=NamedRerouteDeclaration(None=@n42) {name=Input_RoughnessFactor, nodeColor=<rgba 1 0 0.1406 1>}
   MetallicRoughnessTexture_UVs=@n50.UVs
   MetallicRoughnessTexture_DDX=@n50.UVs
   MetallicRoughnessTexture_DDY=@n50.UVs
  ) {materialFunction=/InterchangeAssets/gltf/MaterialFunctions/MF_MetallicRoughness}
  NamedRerouteDeclaration(None=StaticSwitch(
    True=@n51.Roughness
    False=NamedRerouteUsage()
    Value=StaticBoolParameter() {defaultValue=true, group=Configuration, parameterName=bHasMetallicRoughnessTexture, sortPriority=1}
   )) {name=Roughness, nodeColor=<rgba 0 1 0.4609 1>}
  n52 = FunctionInput() {bUsePreviewValueAsDefault=true, inputName=BaseColorFactor_RGB, sortPriority=2, previewValue=<1 1 1 1>}
  n53 = FunctionInput() {bUsePreviewValueAsDefault=true, inputName=BaseColorFactor_A, inputType=FunctionInput_Scalar, sortPriority=3, previewValue=<1 1 1 1>}
  n54 = FunctionInput() {bUsePreviewValueAsDefault=true, inputName=BaseColorTexture_Offset_X, inputType=FunctionInput_Scalar, sortPriority=4, previewValue=<0 0 0 0>}
  n55 = FunctionInput() {bUsePreviewValueAsDefault=true, inputName=BaseColorTexture_Offset_Y, inputType=FunctionInput_Scalar, sortPriority=5, previewValue=<0 0 0 0>}
  n56 = FunctionInput() {bUsePreviewValueAsDefault=true, inputName=BaseColorTexture_Scale_X, inputType=FunctionInput_Scalar, sortPriority=6, previewValue=<1 0 0 0>}
  n57 = FunctionInput() {bUsePreviewValueAsDefault=true, inputName=BaseColorTexture_Scale_Y, inputType=FunctionInput_Scalar, sortPriority=7, previewValue=<1 0 0 0>}
  n58 = FunctionInput() {bUsePreviewValueAsDefault=true, description=Rad, inputName=BaseColorTexture_Rotation, inputType=FunctionInput_Scalar, sortPriority=8}
  n59 = MaterialFunctionCall(
   Offset_X=@n54
   Offset_Y=@n55
   Scale_X=@n56
   Scale_Y=@n57
   Rotation=@n58
   TexCoord=FunctionInput() {bUsePreviewValueAsDefault=true, inputName=BaseColorTexture_TexCoord, inputType=FunctionInput_Scalar, sortPriority=9}
   TilingMethod=FunctionInput() {bUsePreviewValueAsDefault=true, inputName=BaseColorTexture_TilingMethod, inputType=FunctionInput_Vector2, sortPriority=10}
  ) {materialFunction=&MF_TransformUVs}
  n60 = MaterialFunctionCall!(
   BaseColorTexture=FunctionInput(Preview=TextureObject() {texture=&T_White_srgb}) {bUsePreviewValueAsDefault=true, inputName=BaseColorTexture, inputType=FunctionInput_Texture2D, sortPriority=1}
   BaseColorFactor_RGB=NamedRerouteDeclaration(None=@n52) {name=Input_BaseColorFactor_RGB, nodeColor=<rgba 0 1 0.9297 1>}
   BaseColorFactor_A=NamedRerouteDeclaration(None=@n53) {name=Input_BaseColorFactor_A, nodeColor=<rgba 0 1 0.8594 1>}
   BaseColorTexture_UVs=@n59.UVs
   BaseColorTexture_DDX=@n59.UVs
   BaseColorTexture_DDY=@n59.UVs
   AlphaCutoff
  ) {materialFunction=/InterchangeAssets/gltf/MaterialFunctions/MF_BaseColor}
  NamedRerouteDeclaration(None=StaticSwitch(
    True=@n60.BaseColor
    False=NamedRerouteUsage()
    Value=StaticBoolParameter() {defaultValue=true, group=Configuration, parameterName=bHasBaseColorTexture, sortPriority=0}
   )) {name=BaseColor, nodeColor=<rgba 0.6875 1 0 1>}
  NamedRerouteDeclaration(None=StaticSwitch(
    True=@n51.Metallic
    False=NamedRerouteUsage()
    Value=StaticBoolParameter() {defaultValue=true, group=Configuration, parameterName=bHasMetallicRoughnessTexture, sortPriority=1}
   )) {name=Metallic, nodeColor=<rgba 1 0.7734 0 1>}
  MaterialExpressionVertexTangentWS_0 = VertexTangentWS()
  n61 = Custom!(SurfaceTangent=@MaterialExpressionVertexTangentWS_0) {
   code="{\r\n##if MATERIAL_TANGENTSPACENORMAL\r\n  const float3 T = float3(1,0,0);\r\n##else\r\n  const float3 T = normalize(SurfaceTangent);\r\n##endif\r\n  return T;\r\n}"
   inputs [{inputName=SurfaceTangent, input {expression=&self:MaterialExpressionVertexTangentWS_0}}]
  }
  n62 = FunctionInput() {
   bUsePreviewValueAsDefault=true
   inputName=AnisotropyTexture_Offset_X
   inputType=FunctionInput_Scalar
   sortPriority=453
   previewValue=<0 0 0 0>
  }
  n63 = FunctionInput() {
   bUsePreviewValueAsDefault=true
   inputName=AnisotropyTexture_Offset_Y
   inputType=FunctionInput_Scalar
   sortPriority=454
   previewValue=<0 0 0 0>
  }
  n64 = FunctionInput() {
   bUsePreviewValueAsDefault=true
   inputName=AnisotropyTexture_Scale_X
   inputType=FunctionInput_Scalar
   sortPriority=455
   previewValue=<1 0 0 0>
  }
  n65 = FunctionInput() {
   bUsePreviewValueAsDefault=true
   inputName=AnisotropyTexture_Scale_Y
   inputType=FunctionInput_Scalar
   sortPriority=456
   previewValue=<1 0 0 0>
  }
  n66 = FunctionInput() {bUsePreviewValueAsDefault=true, description=Rad, inputName=AnisotropyTexture_Rotation, inputType=FunctionInput_Scalar, sortPriority=457}
  n67 = FunctionInput() {bUsePreviewValueAsDefault=true, description=Rad, inputName=AnisotropyTexture_TexCoord, inputType=FunctionInput_Scalar, sortPriority=458}
  n68 = FunctionInput() {bUsePreviewValueAsDefault=true, inputName=AnisotropyTexture_TilingMethod, inputType=FunctionInput_Vector2, sortPriority=459}
  n69 = MaterialFunctionCall(Offset_X=@n62, Offset_Y=@n63, Scale_X=@n64, Scale_Y=@n65, Rotation=@n66, TexCoord=@n67, TilingMethod=@n68) {materialFunction=&MF_TransformUVs}
  n70 = MaterialFunctionCall!(
   AnisotropyTexture=FunctionInput(Preview=TextureObject() {samplerType=SAMPLERTYPE_LinearColor, texture=/InterchangeAssets/gltf/Textures/T_Anisotropy_Direction_Linear}) {bUsePreviewValueAsDefault=true, inputName=AnisotropyTexture, inputType=FunctionInput_Texture2D, sortPriority=450}
   AnisotropyStrength=FunctionInput() {bUsePreviewValueAsDefault=true, inputName=AnisotropyStrength, inputType=FunctionInput_Scalar, sortPriority=451}
   AnisotropyRotation=FunctionInput() {bUsePreviewValueAsDefault=true, inputName=AnisotropyRotation, inputType=FunctionInput_Scalar, sortPriority=452}
   Normal=NamedRerouteUsage()
   AnisotropyTexture_UVs=@n69.UVs
   AnisotropyTexture_DDX=@n69.UVs
   AnisotropyTexture_DDY=@n69.UVs
  ) {materialFunction=/InterchangeAssets/gltf/MaterialFunctions/MF_SpecularAnisotropy}
  NamedRerouteDeclaration(None=StaticSwitch(
    True=@n70.Tangent
    False=@n61
    Value=StaticBoolParameter() {defaultValue=true, group=Configuration, parameterName=bHasAnisotropyTexture, sortPriority=6}
   )) {name=Tangent, nodeColor=<rgba 1 0.1641 0 1>}
  NamedRerouteDeclaration(None=StaticSwitch(
    True=@n70.Anisotropy
    False=Constant()
    Value=StaticBoolParameter() {defaultValue=true, group=Configuration, parameterName=bHasAnisotropyTexture, sortPriority=6}
   )) {name=Anisotropy, nodeColor=<rgba 0 1 0.2031 1>}
  n71 = FunctionInput() {bUsePreviewValueAsDefault=true, inputName=AlphaMode, inputType=FunctionInput_Scalar, previewValue=<2 0 0 1>}
  MaterialExpressionConstant_2 = Constant() {r=1}
  n72 = StaticSwitch(
   True=@n60.Opacity
   False=NamedRerouteUsage()
   Value=StaticBoolParameter() {defaultValue=true, group=Configuration, parameterName=bHasBaseColorTexture, sortPriority=0}
  )
  MaterialExpressionReroute_0 = Reroute(None=If(
    A=@n72
    B=FunctionInput() {bUsePreviewValueAsDefault=true, inputName=AlphaCutoff, inputType=FunctionInput_Scalar, sortPriority=11}
    "A > B"=@MaterialExpressionConstant_2
    "A < B"=Constant()
   ))
  MaterialExpressionReroute_2 = Reroute(None=@n72)
  n73 = Switch!(SwitchValue=@n71, Default, Opaque=@MaterialExpressionConstant_2, Mask=@MaterialExpressionReroute_0, Blend=@MaterialExpressionReroute_2) {
   constDefault=2
   inputs [
    {
     inputName=Opaque
     input {
      expression=&self:MaterialExpressionConstant_2
      inputName=None
      mask=0
      maskA=0
      maskB=0
      maskG=0
      maskR=0
      outputIndex=0
     }
    }
    {
     inputName=Mask
     input {
      expression=&self:MaterialExpressionReroute_0
      inputName=None
      mask=0
      maskA=0
      maskB=0
      maskG=0
      maskR=0
      outputIndex=0
     }
    }
    {
     inputName=Blend
     input {
      expression=&self:MaterialExpressionReroute_2
      inputName=None
      mask=0
      maskA=0
      maskB=0
      maskG=0
      maskR=0
      outputIndex=0
     }
    }
   ]
  }
  NamedRerouteDeclaration(None=@n73) {name=AlphaCoverage, nodeColor=<rgba 0.8516 1 0 1>}
  n74 = NamedRerouteUsage()
  n75 = NamedRerouteUsage()
  n76 = Constant() {r=5}
  NamedRerouteDeclaration(None=OneMinus(None=MaterialFunctionCall!(Normals TS=@n74, Specular=@n75, FresnelExponent=@n76) {materialFunction=/InterchangeAssets/gltf/MaterialFunctions/MF_Fresnel_DS}.Fresnel)) {name=InverseFresnel, nodeColor=<rgba 0.4766 0 1 1>}
  n77 = FunctionInput(Preview=TextureObject() {samplerType=SAMPLERTYPE_Normal, texture=/InterchangeAssets/gltf/Textures/T_Generic_N}) {bUsePreviewValueAsDefault=true, inputName=NormalTexture, inputType=FunctionInput_Texture2D, sortPriority=100}
  n78 = FunctionInput() {bUsePreviewValueAsDefault=true, inputName=NormalScale, inputType=FunctionInput_Scalar, sortPriority=101, previewValue=<1 0 0 1>}
  n79 = NamedRerouteDeclaration(None=@n78) {name=Input_NormalScale, nodeColor=<rgba 0.5703 1 0 1>}
  n80 = FunctionInput() {bUsePreviewValueAsDefault=true, inputName=NormalTexture_Offset_X, inputType=FunctionInput_Scalar, sortPriority=102, previewValue=<0 0 0 0>}
  n81 = FunctionInput() {bUsePreviewValueAsDefault=true, inputName=NormalTexture_Offset_Y, inputType=FunctionInput_Scalar, sortPriority=103, previewValue=<0 0 0 0>}
  n82 = FunctionInput() {bUsePreviewValueAsDefault=true, inputName=NormalTexture_Scale_X, inputType=FunctionInput_Scalar, sortPriority=104, previewValue=<1 0 0 0>}
  n83 = FunctionInput() {bUsePreviewValueAsDefault=true, inputName=NormalTexture_Scale_Y, inputType=FunctionInput_Scalar, sortPriority=105, previewValue=<1 0 0 0>}
  n84 = FunctionInput() {bUsePreviewValueAsDefault=true, description=Rad, inputName=NormalTexture_Rotation, inputType=FunctionInput_Scalar, sortPriority=106}
  n85 = MaterialFunctionCall(
   Offset_X=@n80
   Offset_Y=@n81
   Scale_X=@n82
   Scale_Y=@n83
   Rotation=@n84
   TexCoord=FunctionInput() {bUsePreviewValueAsDefault=true, inputName=NormalTexture_TexCoord, inputType=FunctionInput_Scalar, sortPriority=107}
   TilingMethod=FunctionInput() {bUsePreviewValueAsDefault=true, inputName=NormalTexture_TilingMethod, inputType=FunctionInput_Vector2, sortPriority=108}
  ) {materialFunction=&MF_TransformUVs}
  n86 = Constant3Vector() {constant=<rgba 0 0 1 0>}
  n87 = NamedRerouteUsage()
  NamedRerouteDeclaration(None=StaticSwitchParameter(
    True=MaterialFunctionCall!(
     NormalTexture=@n77
     NormalScale=@n79
     NormalTexture_UVs=@n85.UVs
     NormalTexture_DDX=@n85.UVs
     NormalTexture_DDY=@n85.UVs
     NormalTexture_SinCos=@n85.UVs
    ) {materialFunction=/InterchangeAssets/gltf/MaterialFunctions/MF_Normals}.Normals
    False=MaterialFunctionCall!(Normal=@n86, Scale=@n87) {materialFunction=/InterchangeAssets/gltf/MaterialFunctions/MF_ScaleNormal}.Result
   ) {defaultValue=true, group=Configuration, parameterName=bHasNormalTexture, sortPriority=3}) {name=Normals, nodeColor=<rgba 0.2188 1 0 1>}
  n88 = FunctionInput() {
   bUsePreviewValueAsDefault=true
   inputName=TransmissionTexture_Offset_X
   inputType=FunctionInput_Scalar
   sortPriority=352
   previewValue=<0 0 0 0>
  }
  n89 = FunctionInput() {
   bUsePreviewValueAsDefault=true
   inputName=TransmissionTexture_Offset_Y
   inputType=FunctionInput_Scalar
   sortPriority=353
   previewValue=<0 0 0 0>
  }
  n90 = FunctionInput() {
   bUsePreviewValueAsDefault=true
   inputName=TransmissionTexture_Scale_X
   inputType=FunctionInput_Scalar
   sortPriority=354
   previewValue=<1 0 0 0>
  }
  n91 = FunctionInput() {
   bUsePreviewValueAsDefault=true
   inputName=TransmissionTexture_Scale_Y
   inputType=FunctionInput_Scalar
   sortPriority=355
   previewValue=<1 0 0 0>
  }
  n92 = FunctionInput() {bUsePreviewValueAsDefault=true, description=Rad, inputName=TransmissionTexture_Rotation, inputType=FunctionInput_Scalar, sortPriority=356}
  n93 = FunctionInput() {bUsePreviewValueAsDefault=true, inputName=TransmissionTexture_TilingMethod, inputType=FunctionInput_Vector2, sortPriority=358}
  n94 = MaterialFunctionCall(
   Offset_X=@n88
   Offset_Y=@n89
   Scale_X=@n90
   Scale_Y=@n91
   Rotation=@n92
   TexCoord=FunctionInput() {bUsePreviewValueAsDefault=true, inputName=TransmissionTexture_TexCoord, inputType=FunctionInput_Scalar, sortPriority=357}
   TilingMethod=@n93
  ) {materialFunction=&MF_TransformUVs}
  n95 = MaterialFunctionCall!(
   TransmissionTexture=FunctionInput(Preview=TextureObject() {samplerType=SAMPLERTYPE_LinearColor, texture=&T_White_Linear}) {bUsePreviewValueAsDefault=true, inputName=TransmissionTexture, inputType=FunctionInput_Texture2D, sortPriority=350}
   TransmissionFactor=NamedRerouteDeclaration(None=FunctionInput() {bUsePreviewValueAsDefault=true, inputName=TransmissionFactor, inputType=FunctionInput_Scalar, sortPriority=351}) {name=Input_TransmissionFactor, nodeColor=<rgba 1 0 0.1641 1>}
   TransmissionTexture_UVs=@n94.UVs
   TransmissionTexture_DDX=@n94.UVs
   TransmissionTexture_DDY=@n94.UVs
  ) {materialFunction=/InterchangeAssets/gltf/MaterialFunctions/MF_Transmission}
  n96 = NamedRerouteUsage()
  NamedRerouteDeclaration(None=Multiply(
    A=StaticSwitchParameter(True=@n95.Transmission, False=@n96) {defaultValue=true, group=Configuration, parameterName=bHasTransmissionTexture, sortPriority=8}
    B=NamedRerouteUsage()
   )) {name=Transmission, nodeColor=<rgba 0 1 0.2734 1>}
  n97 = FunctionInput(Preview=TextureObject() {samplerType=SAMPLERTYPE_LinearColor, texture=&T_White_Linear}) {bUsePreviewValueAsDefault=true, inputName=OcclusionTexture, inputType=FunctionInput_Texture2D, sortPriority=151}
  n98 = FunctionInput() {bUsePreviewValueAsDefault=true, inputName=OcclusionStrength, inputType=FunctionInput_Scalar, sortPriority=152, previewValue=<1 0 0 1>}
  n99 = NamedRerouteDeclaration(None=@n98) {name=Input_OcclusionStrength, nodeColor=<rgba 1 0.4453 0 1>}
  n100 = FunctionInput() {
   bUsePreviewValueAsDefault=true
   inputName=OcclusionTexture_Offset_X
   inputType=FunctionInput_Scalar
   sortPriority=153
   previewValue=<0 0 0 0>
  }
  n101 = FunctionInput() {
   bUsePreviewValueAsDefault=true
   inputName=OcclusionTexture_Offset_Y
   inputType=FunctionInput_Scalar
   sortPriority=154
   previewValue=<0 0 0 0>
  }
  n102 = FunctionInput() {bUsePreviewValueAsDefault=true, inputName=OcclusionTexture_Scale_X, inputType=FunctionInput_Scalar, sortPriority=155, previewValue=<1 0 0 0>}
  n103 = FunctionInput() {bUsePreviewValueAsDefault=true, inputName=OcclusionTexture_Scale_Y, inputType=FunctionInput_Scalar, sortPriority=156, previewValue=<1 0 0 0>}
  n104 = FunctionInput() {bUsePreviewValueAsDefault=true, description=Rad, inputName=OcclusionTexture_Rotation, inputType=FunctionInput_Scalar, sortPriority=157}
  n105 = FunctionInput() {bUsePreviewValueAsDefault=true, description=Rad, inputName=OcclusionTexture_TexCoord, inputType=FunctionInput_Scalar, sortPriority=158}
  n106 = FunctionInput() {bUsePreviewValueAsDefault=true, inputName=OcclusionTexture_TilingMethod, inputType=FunctionInput_Vector2, sortPriority=159}
  n107 = MaterialFunctionCall(Offset_X=@n100, Offset_Y=@n101, Scale_X=@n102, Scale_Y=@n103, Rotation=@n104, TexCoord=@n105, TilingMethod=@n106) {materialFunction=&MF_TransformUVs}
  NamedRerouteDeclaration(None=StaticSwitchParameter(
    True=MaterialFunctionCall!(
     OcclusionTexture=@n97
     OcclusionStrength=@n99
     OcclusionTexture_UVs=@n107.UVs
     OcclusionTexture_DDX=@n107.UVs
     OcclusionTexture_DDY=@n107.UVs
    ) {materialFunction=/InterchangeAssets/gltf/MaterialFunctions/MF_Occlusion}.Occlusion
    False=NamedRerouteUsage()
   ) {defaultValue=true, group=Configuration, parameterName=bHasOcclusionTexture, sortPriority=4}) {name=Occlusion, nodeColor=<rgba 0.9453 0 1 1>}
  n108 = FunctionInput() {bUsePreviewValueAsDefault=true, inputName=IOR, inputType=FunctionInput_Scalar, sortPriority=209, previewValue=<1.5 0 0 1>}
  n109 = FunctionInput(Preview=TextureObject() {texture=&T_White_srgb}) {bUsePreviewValueAsDefault=true, inputName=SpecularTexture, inputType=FunctionInput_Texture2D, sortPriority=200}
  n110 = FunctionInput() {bUsePreviewValueAsDefault=true, inputName=SpecularFactor, inputType=FunctionInput_Scalar, sortPriority=201, previewValue=<0.5 0 0 1>}
  n111 = NamedRerouteDeclaration(None=@n110) {name=Input_SpecularFactor, nodeColor=<rgba 0 0.7422 1 1>}
  n112 = FunctionInput() {bUsePreviewValueAsDefault=true, inputName=SpecularTexture_Offset_X, inputType=FunctionInput_Scalar, sortPriority=202, previewValue=<0 0 0 0>}
  n113 = FunctionInput() {bUsePreviewValueAsDefault=true, inputName=SpecularTexture_Offset_Y, inputType=FunctionInput_Scalar, sortPriority=203, previewValue=<0 0 0 0>}
  n114 = FunctionInput() {bUsePreviewValueAsDefault=true, inputName=SpecularTexture_Scale_X, inputType=FunctionInput_Scalar, sortPriority=204, previewValue=<1 0 0 0>}
  n115 = FunctionInput() {bUsePreviewValueAsDefault=true, inputName=SpecularTexture_Scale_Y, inputType=FunctionInput_Scalar, sortPriority=205, previewValue=<1 0 0 0>}
  n116 = FunctionInput() {bUsePreviewValueAsDefault=true, description=Rad, inputName=SpecularTexture_Rotation, inputType=FunctionInput_Scalar, sortPriority=206}
  n117 = FunctionInput() {bUsePreviewValueAsDefault=true, description=Rad, inputName=SpecularTexture_TexCoord, inputType=FunctionInput_Scalar, sortPriority=207}
  n118 = MaterialFunctionCall(
   Offset_X=@n112
   Offset_Y=@n113
   Scale_X=@n114
   Scale_Y=@n115
   Rotation=@n116
   TexCoord=@n117
   TilingMethod=FunctionInput() {bUsePreviewValueAsDefault=true, inputName=SpecularTexture_TilingMethod, inputType=FunctionInput_Vector2, sortPriority=208}
  ) {materialFunction=&MF_TransformUVs}
  n119 = Multiply(
   A=MaterialFunctionCall!(IOR=NamedRerouteDeclaration(None=@n108) {name=IOR, nodeColor=<rgba 0 0.5312 1 1>}) {materialFunction=/InterchangeAssets/gltf/MaterialFunctions/MF_IOR}.Specular
   B=StaticSwitchParameter(
    True=MaterialFunctionCall!(SpecularTexture=@n109, SpecularFactor=@n111, SpecularTexture_UVs=@n118.UVs, SpecularTexture_DDX=@n118.UVs, SpecularTexture_DDY=@n118.UVs) {materialFunction=/InterchangeAssets/gltf/MaterialFunctions/MF_Specular}.Specular
    False=NamedRerouteUsage()
   ) {defaultValue=true, group=Configuration, parameterName=bHasSpecularTexture, sortPriority=5}
  )
  n120 = NamedRerouteUsage()
  NamedRerouteDeclaration(None=Multiply(A=Multiply(A=@n119, B=@n120), B=NamedRerouteUsage())) {name=Specular, nodeColor=<rgba 0.3125 0 1 1>}
 ]
}
