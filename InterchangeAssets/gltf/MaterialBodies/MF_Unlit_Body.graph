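# Material function forming the body of unlit glTF materials in InterchangeAssets, outputting UnlitColor, Opacity and OpacityMask.
# MF_TransformUVs applies offset, scale, rotation, texcoord and tiling inputs, feeding MF_BaseColor along with the texture, color and alpha factors, and alpha cutoff.
# Static switches on bHasBaseColorTexture choose between those textured results and untextured fallbacks through named reroutes, using a subtraction for the mask.
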MaterialFunction /InterchangeAssets/gltf/MaterialBodies/MF_Unlit_Body
graph {nodes [
  n1 = FunctionInput() {bUsePreviewValueAsDefault=true, inputName=BaseColorFactor_RGB, sortPriority=1, previewValue=<1 1 1 1>}
  n2 = FunctionInput() {bUsePreviewValueAsDefault=true, inputName=BaseColorFactor_A, inputType=FunctionInput_Scalar, sortPriority=2, previewValue=<1 1 1 1>}
  n3 = FunctionInput() {bUsePreviewValueAsDefault=true, inputName=BaseColorTexture_Offset_X, inputType=FunctionInput_Scalar, sortPriority=3, previewValue=<0 0 0 0>}
  n4 = FunctionInput() {bUsePreviewValueAsDefault=true, inputName=BaseColorTexture_Offset_Y, inputType=FunctionInput_Scalar, sortPriority=4, previewValue=<0 0 0 0>}
  n5 = FunctionInput() {bUsePreviewValueAsDefault=true, inputName=BaseColorTexture_Scale_X, inputType=FunctionInput_Scalar, sortPriority=5, previewValue=<1 0 0 0>}
  n6 = FunctionInput() {bUsePreviewValueAsDefault=true, inputName=BaseColorTexture_Scale_Y, inputType=FunctionInput_Scalar, sortPriority=6, previewValue=<1 0 0 0>}
  n7 = FunctionInput() {bUsePreviewValueAsDefault=true, description=Rad, inputName=BaseColorTexture_Rotation, inputType=FunctionInput_Scalar, sortPriority=7}
  n8 = MaterialFunctionCall!(
   Offset_X=@n3
   Offset_Y=@n4
   Scale_X=@n5
   Scale_Y=@n6
   Rotation=@n7
   TexCoord=FunctionInput() {bUsePreviewValueAsDefault=true, inputName=BaseColorTexture_TexCoord, inputType=FunctionInput_Scalar, sortPriority=8}
   TilingMethod=FunctionInput() {bUsePreviewValueAsDefault=true, inputName=BaseColorTexture_TilingMethod, inputType=FunctionInput_Vector2, sortPriority=9}
  ) {materialFunction=/InterchangeAssets/gltf/MaterialFunctions/MF_TransformUVs}
  n9 = MaterialFunctionCall!(
   BaseColorTexture=FunctionInput(Preview=TextureObject() {texture=/InterchangeAssets/gltf/Textures/T_White_srgb}) {bUsePreviewValueAsDefault=true, inputName=BaseColorTexture, inputType=FunctionInput_Texture2D}
   BaseColorFactor_RGB=NamedRerouteDeclaration(None=@n1) {name=Input_BaseColorFactor_RGB, nodeColor=<rgba 1 0.8203 0 1>}
   BaseColorFactor_A=NamedRerouteDeclaration(None=@n2) {name=Input_BaseColorFactor_A, nodeColor=<rgba 0.8984 0 1 1>}
   BaseColorTexture_UVs=@n8.UVs
   BaseColorTexture_DDX=@n8.UVs
   BaseColorTexture_DDY=@n8.UVs
   AlphaCutoff=NamedRerouteDeclaration(None=FunctionInput() {bUsePreviewValueAsDefault=true, inputName=AlphaCutoff, inputType=FunctionInput_Scalar, sortPriority=10}) {name=Input_AlphaCutoff, nodeColor=<rgba 1 0.7734 0 1>}
  ) {materialFunction=/InterchangeAssets/gltf/MaterialFunctions/MF_BaseColor}
  FunctionOutput(None=StaticSwitch(
    True=@n9.BaseColor
    False=NamedRerouteUsage()
    Value=StaticBoolParameter() {defaultValue=true, group=Configuration, parameterName=bHasBaseColorTexture, sortPriority=0}
   )) {outputName=UnlitColor, sortPriority=1}
  n10 = NamedRerouteUsage()
  n11 = NamedRerouteUsage()
  FunctionOutput(None=StaticSwitch(
    True=@n9.OpacityMask
    False=Subtract(A=@n10, B=@n11)
    Value=StaticBoolParameter() {defaultValue=true, group=Configuration, parameterName=bHasBaseColorTexture, sortPriority=0}
   )) {outputName=OpacityMask, sortPriority=7}
  FunctionOutput(None=StaticSwitch(
    True=@n9.Opacity
    False=NamedRerouteUsage()
    Value=StaticBoolParameter() {defaultValue=true, group=Configuration, parameterName=bHasBaseColorTexture, sortPriority=0}
   )) {outputName=Opacity, sortPriority=6}
 ]}
