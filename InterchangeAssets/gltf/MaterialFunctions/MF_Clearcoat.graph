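# Material function for glTF clear coat in Interchange materials, outputting clear coat amount, roughness and top normal.
# Amount is the texture's red channel times its factor; roughness is the roughness texture's green channel times its own factor.
# The normal map is rotated by a SinCos input, then flattened by one minus the saturated normal scale.
# All three textures sample with caller-supplied UVs and explicit derivatives through shared clamp samplers.
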MaterialFunction /InterchangeAssets/gltf/MaterialFunctions/MF_Clearcoat
graph {nodes [
  n1 = FunctionInput() {bUsePreviewValueAsDefault=true, inputName=ClearCoatRoughnessTexture_UVs, inputType=FunctionInput_Vector2, sortPriority=9}
  n2 = FunctionInput(Preview=TextureObject() {samplerType=SAMPLERTYPE_LinearColor, texture=/InterchangeAssets/gltf/Textures/T_White_Linear}) {bUsePreviewValueAsDefault=true, inputName=ClearCoatRoughnessTexture, inputType=FunctionInput_Texture2D, sortPriority=1}
  n3 = FunctionInput() {bUsePreviewValueAsDefault=true, inputName=ClearCoatRoughnessTexture_DDX, inputType=FunctionInput_Vector2, sortPriority=10}
  n4 = FunctionInput() {bUsePreviewValueAsDefault=true, inputName=ClearCoatRoughnessTexture_DDY, inputType=FunctionInput_Vector2, sortPriority=11}
  n5 = FunctionInput() {bUsePreviewValueAsDefault=true, inputName=ClearCoatRoughnessFactor, inputType=FunctionInput_Scalar, sortPriority=4, previewValue=<0 0 0 0>}
  FunctionOutput(None=Multiply(
    A=TextureSample!(UVs=@n1, Tex=@n2, "DDX(UVs)"=@n3, "DDY(UVs)"=@n4, Apply View MipBias) {mipValueMode=TMVM_Derivative, samplerSource=SSM_Clamp_WorldGroupSettings, samplerType=SAMPLERTYPE_LinearColor}.G
    B=@n5
   )) {outputName=ClearCoatRoughness, sortPriority=1}
  n6 = FunctionInput() {bUsePreviewValueAsDefault=true, inputName=ClearCoatTexture_UVs, inputType=FunctionInput_Vector2, sortPriority=6}
  n7 = FunctionInput(Preview=TextureObject() {samplerType=SAMPLERTYPE_LinearColor, texture=&T_White_Linear}) {bUsePreviewValueAsDefault=true, inputName=ClearCoatTexture, inputType=FunctionInput_Texture2D}
  n8 = FunctionInput() {bUsePreviewValueAsDefault=true, inputName=ClearCoatTexture_DDX, inputType=FunctionInput_Vector2, sortPriority=7}
  n9 = FunctionInput() {bUsePreviewValueAsDefault=true, inputName=ClearCoatTexture_DDY, inputType=FunctionInput_Vector2, sortPriority=8}
  n10 = FunctionInput() {bUsePreviewValueAsDefault=true, inputName=ClearCoatFactor, inputType=FunctionInput_Scalar, sortPriority=3, previewValue=<0 0 0 0>}
  FunctionOutput(None=Multiply(
    A=TextureSample(UVs=@n6, Tex=@n7, "DDX(UVs)"=@n8, "DDY(UVs)"=@n9) {mipValueMode=TMVM_Derivative, samplerSource=SSM_Clamp_WorldGroupSettings, samplerType=SAMPLERTYPE_LinearColor}.R
    B=@n10
   )) {outputName=ClearCoatAmount}
  n11 = FunctionInput(Preview=TextureObject() {samplerType=SAMPLERTYPE_Normal, texture=/InterchangeAssets/gltf/Textures/T_Generic_N}) {bUsePreviewValueAsDefault=true, desc=Top Normal, inputName=ClearCoatNormalTexture, inputType=FunctionInput_Texture2D, sortPriority=2}
  n12 = TextureSample(
   UVs=FunctionInput() {bUsePreviewValueAsDefault=true, inputName=ClearCoatNormalTexture_UVs, inputType=FunctionInput_Vector2, sortPriority=12}
   Tex=@n11
   "DDX(UVs)"=FunctionInput() {bUsePreviewValueAsDefault=true, inputName=ClearCoatNormalTexture_DDX, inputType=FunctionInput_Vector2, sortPriority=13}
   "DDY(UVs)"=FunctionInput() {bUsePreviewValueAsDefault=true, inputName=ClearCoatNormalTexture_DDY, inputType=FunctionInput_Vector2, sortPriority=14}
  ) {mipValueMode=TMVM_Derivative, samplerSource=SSM_Clamp_WorldGroupSettings, samplerType=SAMPLERTYPE_Normal}
  n13 = FunctionInput() {
   bUsePreviewValueAsDefault=true
   inputName=ClearCoatNormalTexture_SinCos
   inputType=FunctionInput_Vector2
   sortPriority=15
   previewValue=<0 1 0 1>
  }
  n14 = FunctionInput() {bUsePreviewValueAsDefault=true, inputName=ClearCoatNormalScale, inputType=FunctionInput_Scalar, sortPriority=5, previewValue=<1 0 0 0>}
  n15 = Saturate(None=@n14)
  FunctionOutput(None=MaterialFunctionCall(
    Normal=MaterialFunctionCall!(NormalsTS=@n12.RGB, SinCos=@n13) {materialFunction=&/MF_RotateNormals_TS}.Result
    Flatness=OneMinus(None=@n15)
   ) {materialFunction=&FlattenNormal}.Result) {outputName=ClearCoatTopNormal, sortPriority=2}
 ]}
