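# Material function for glTF materials in the Interchange assets that evaluates specular-glossiness inputs into DiffuseColor, SpecularColor, Glossiness and Opacity.
# The diffuse texture times its RGB and alpha factors gives color and opacity.
# The specular-glossiness texture times its factors gives specular from RGB and glossiness from alpha.
# Each texture takes its own UVs and explicit DDX/DDY derivatives, uses shared clamp samplers, and defaults to white.
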MaterialFunction /InterchangeAssets/gltf/MaterialFunctions/MF_DiffuseSpecGloss
graph {nodes [
  n1 = TextureSample!(
   UVs=FunctionInput() {bUsePreviewValueAsDefault=true, inputName=DiffuseTexture_UVs, inputType=FunctionInput_Vector2, sortPriority=6}
   Tex=FunctionInput(Preview=TextureObject() {texture=/InterchangeAssets/gltf/Textures/T_White_srgb}) {bUsePreviewValueAsDefault=true, inputName=DiffuseTexture, inputType=FunctionInput_Texture2D}
   "DDX(UVs)"=FunctionInput() {bUsePreviewValueAsDefault=true, inputName=DiffuseTexture_DDX, inputType=FunctionInput_Vector2, sortPriority=7}
   "DDY(UVs)"=FunctionInput() {bUsePreviewValueAsDefault=true, inputName=DiffuseTexture_DDY, inputType=FunctionInput_Vector2, sortPriority=8}
   Apply View MipBias
  ) {mipValueMode=TMVM_Derivative, samplerSource=SSM_Clamp_WorldGroupSettings}
  n2 = FunctionInput() {bUsePreviewValueAsDefault=true, inputName=DiffuseFactor_RGB, sortPriority=2, previewValue=<1 1 1 0>}
  FunctionOutput(None=Multiply(A=@n1.RGB, B=@n2)) {outputName=DiffuseColor}
  n3 = TextureSample(
   UVs=FunctionInput() {bUsePreviewValueAsDefault=true, inputName=SpecularGlossinessTexture_UVs, inputType=FunctionInput_Vector2, sortPriority=9}
   Tex=FunctionInput(Preview=TextureObject() {texture=&T_White_srgb}) {bUsePreviewValueAsDefault=true, inputName=SpecularGlossinessTexture, inputType=FunctionInput_Texture2D, sortPriority=1}
   "DDX(UVs)"=FunctionInput() {bUsePreviewValueAsDefault=true, inputName=SpecularGlossinessTexture_DDX, inputType=FunctionInput_Vector2, sortPriority=10}
   "DDY(UVs)"=FunctionInput() {bUsePreviewValueAsDefault=true, inputName=SpecularGlossinessTexture_DDY, inputType=FunctionInput_Vector2, sortPriority=11}
  ) {mipValueMode=TMVM_Derivative, samplerSource=SSM_Clamp_WorldGroupSettings}
  FunctionOutput(None=Multiply(A=@n3.RGB, B=FunctionInput() {bUsePreviewValueAsDefault=true, inputName=SpecFactor, sortPriority=4, previewValue=<1 1 1 0>})) {outputName=SpecularColor, sortPriority=1}
  n4 = FunctionInput() {bUsePreviewValueAsDefault=true, inputName=GlossinessFactor, inputType=FunctionInput_Scalar, sortPriority=5, previewValue=<1 0 0 0>}
  FunctionOutput(None=Multiply(A=@n3.A, B=@n4)) {outputName=Glossiness, sortPriority=2}
  n5 = FunctionInput() {bUsePreviewValueAsDefault=true, inputName=DiffuseFactor_A, inputType=FunctionInput_Scalar, sortPriority=3, previewValue=<1 0 0 0>}
  FunctionOutput(None=Multiply(A=@n1.A, B=@n5)) {outputName=Opacity, sortPriority=3}
 ]}
