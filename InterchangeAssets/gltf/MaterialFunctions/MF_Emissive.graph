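# MF_Emissive is a material function among the Interchange glTF assets that computes an emissive colour.
# It samples the emissive texture with caller-supplied UVs and explicit DDX/DDY derivatives through the shared world-group clamp sampler.
# The sampled RGB is multiplied by EmissiveFactor and EmissiveStrength to give the Emissive output.
# Defaults are a white sRGB texture, zero factor and unit strength, so emission is black unless a factor is supplied.
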MaterialFunction /InterchangeAssets/gltf/MaterialFunctions/MF_Emissive
graph {nodes [
  n1 = TextureSample!(
   UVs=FunctionInput() {bUsePreviewValueAsDefault=true, inputName=EmissiveTexture_UVs, inputType=FunctionInput_Vector2, sortPriority=3}
   Tex=FunctionInput(Preview=TextureObject() {texture=/InterchangeAssets/gltf/Textures/T_White_srgb}) {bUsePreviewValueAsDefault=true, inputName=EmissiveTexture, inputType=FunctionInput_Texture2D}
   "DDX(UVs)"=FunctionInput() {bUsePreviewValueAsDefault=true, inputName=EmissiveTexture_DDX, inputType=FunctionInput_Vector2, sortPriority=4}
   "DDY(UVs)"=FunctionInput() {bUsePreviewValueAsDefault=true, inputName=EmissiveTexture_DDY, inputType=FunctionInput_Vector2, sortPriority=5}
   Apply View MipBias
  ) {mipValueMode=TMVM_Derivative, samplerSource=SSM_Clamp_WorldGroupSettings}
  n2 = FunctionInput() {bUsePreviewValueAsDefault=true, inputName=EmissiveFactor, sortPriority=1, previewValue=<0 0 0 0>}
  n3 = FunctionInput() {bUsePreviewValueAsDefault=true, inputName=EmissiveStrength, inputType=FunctionInput_Scalar, sortPriority=2, previewValue=<1 0 0 0>}
  FunctionOutput(None=Multiply(A=Multiply(A=@n1.RGB, B=@n2), B=@n3)) {outputName=Emissive}
 ]}
