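# Material function in the glTF Interchange assets, named for returning the largest component of a Float3 input.
# It splits the input with BreakOutFloat3Components and feeds the channels into two nested Max nodes.
# The inner Max takes R on both sides, so the output is the larger of R and B, and G is never used.
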MaterialFunction /InterchangeAssets/gltf/MaterialFunctions/MF_MaxComponentValue
graph {nodes [
  n1 = MaterialFunctionCall(Float3=FunctionInput()) {materialFunction=&BreakOutFloat3Components}
  FunctionOutput(None=Max(A=Max(A=@n1.R, B=@n1.R), B=@n1.B))
 ]}
